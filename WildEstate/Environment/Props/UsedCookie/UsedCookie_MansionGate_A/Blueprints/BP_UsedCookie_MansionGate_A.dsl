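; Openable mansion gate prop that swings open automatically when a player drives a vehicle into it at speed.
; On the server, an overlap on the vehicle trigger box checks that the actor is a player pawn in a vehicle moving above a cached-speed threshold.
; If the gate is not already open, it calls OpenActor with that pawn's controller.
; BeginPlay, Tick and actor overlap events are empty, and the construction script only calls its parent.
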(event EventBeginPlay)

(event Collision|EventActorBeginOverlap
  (params (OtherActor :type Actor%20Object%20Reference)))

(event EventTick
  (params (DeltaSeconds :type Float%20%28single-precision%29)))

(event OnComponentBeginOverlap(VehicleOVerlapBox)
  (params (OverlappedComponent :type Primitive%20Component%20Object%20Reference) (OtherActor :type Actor%20Object%20Reference) (OtherComp :type Primitive%20Component%20Object%20Reference) (OtherBodyIndex :type Integer) (bFromSweep :type Boolean) (SweepResult :type Hit%20Result%20Structure%20%28by%20ref%29))
  (Networking|SwitchHasAuthority
    (:Authority
      (bind _asfort_player_pawn_athena (Utilities|Casting|CastToFortPlayerPawnAthena OtherActor))
      (if (Vehicle|IsinVehicle _asfort_player_pawn_athena)
        (if (> (Variables|/Script/FortniteGame.FortAthenaVehicle|GetCachedSpeed (Vehicle|GetVehicle _asfort_player_pawn_athena)) 6e+02)
          (if (not (Helper|IsDoorOpen))
            (Openable|OpenActor self (Fortnite|GetControllerfromActor _asfort_player_pawn_athena) true)))))))

(fn ConstructionScript
  (|CallParentFunction))
